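# Sound cue for tree wood-creak ambient spots in the Dark Forest POI, played quietly with distant-creak attenuation and an exterior ambient bus send.
# A quality-level node picks between 18 random creak variants for its first two levels and a four-variant subset for the last two.
# Two stages of pitch and volume modulation vary each play, and a root enveloper eases volume down after a few seconds.
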SoundCue /Game/Sounds/Helios/POI/DarkForest/Spots/Spot_DarkForest_Tree_Creak_Long_Looping_Cue
root=Enveloper_0
nodes [
 {
  id=Enveloper_0
  children [Modulator_13]
  properties {pitchCurve {editorCurveData {keys [{time=0}]}}, volumeCurve {editorCurveData {keys [<at 0 1>,<at 3.5 1>,<at 7 0.7332>]}}}
 }
 {id=Modulator_0, children [Random_4], properties {pitchMax=1, pitchMin=0.7, volumeMax=1.1, volumeMin=0.5}}
 {id=Modulator_1, children [Random_2], properties {pitchMax=1, pitchMin=0.7, volumeMax=1.1, volumeMin=0.5}}
 {id=Modulator_13, children [QualityLevel_0], properties {pitchMax=1.1, pitchMin=0.9, volumeMax=1.1, volumeMin=0.9}}
 {id=QualityLevel_0, children [Modulator_1,Modulator_1,Modulator_0,Modulator_0]}
 {
  id=Random_2
  children [
   WavePlayer_20
   WavePlayer_21
   WavePlayer_22
   WavePlayer_23
   WavePlayer_24
   WavePlayer_25
   WavePlayer_26
   WavePlayer_27
   WavePlayer_28
   WavePlayer_29
   WavePlayer_30
   WavePlayer_31
   WavePlayer_32
   WavePlayer_33
   WavePlayer_34
   WavePlayer_35
   WavePlayer_36
   WavePlayer_37
  ]
  properties {weights [1,1,1,1,1,1,1,1,1,1,1,1,1,1,1,1,1,1]}
 }
 {
  id=Random_4
  children [WavePlayer_56, WavePlayer_57, WavePlayer_58, WavePlayer_59]
  properties {bShouldExcludeFromBranchCulling=true, weights [1,1,1,1]}
 }
 {id=WavePlayer_20, properties {soundWaveAssetPtr=&/Spot_DarkForest_Trees_Wood_Creaks_01}}
 {id=WavePlayer_21, properties {soundWaveAssetPtr=&/Spot_DarkForest_Trees_Wood_Creaks_02}}
 {id=WavePlayer_22, properties {soundWaveAssetPtr=&/Spot_DarkForest_Trees_Wood_Creaks_03}}
 {id=WavePlayer_23, properties {soundWaveAssetPtr=&/Spot_DarkForest_Trees_Wood_Creaks_04}}
 {id=WavePlayer_24, properties {soundWaveAssetPtr=&/Spot_DarkForest_Trees_Wood_Creaks_05}}
 {id=WavePlayer_25, properties {soundWaveAssetPtr=&/Spot_DarkForest_Trees_Wood_Creaks_06}}
 {id=WavePlayer_26, properties {soundWaveAssetPtr=&/Spot_DarkForest_Trees_Wood_Creaks_07}}
 {id=WavePlayer_27, properties {soundWaveAssetPtr=&/Spot_DarkForest_Trees_Wood_Creaks_08}}
 {id=WavePlayer_28, properties {soundWaveAssetPtr=&/Spot_DarkForest_Trees_Wood_Creaks_09}}
 {id=WavePlayer_29, properties {soundWaveAssetPtr=&/Spot_DarkForest_Trees_Wood_Creaks_10}}
 {id=WavePlayer_30, properties {soundWaveAssetPtr=&/Spot_DarkForest_Trees_Wood_Creaks_11}}
 {id=WavePlayer_31, properties {soundWaveAssetPtr=&/Spot_DarkForest_Trees_Wood_Creaks_12}}
 {id=WavePlayer_32, properties {soundWaveAssetPtr=&/Spot_DarkForest_Trees_Wood_Creaks_13}}
 {id=WavePlayer_33, properties {soundWaveAssetPtr=&/Spot_DarkForest_Trees_Wood_Creaks_14}}
 {id=WavePlayer_34, properties {soundWaveAssetPtr=&/Spot_DarkForest_Trees_Wood_Creaks_15}}
 {id=WavePlayer_35, properties {soundWaveAssetPtr=&/Spot_DarkForest_Trees_Wood_Creaks_16}}
 {id=WavePlayer_36, properties {soundWaveAssetPtr=&/Spot_DarkForest_Trees_Wood_Creaks_17}}
 {id=WavePlayer_37, properties {soundWaveAssetPtr=&/Spot_DarkForest_Trees_Wood_Creaks_18}}
 {id=WavePlayer_56, properties {soundWaveAssetPtr=&Spot_DarkForest_Trees_Wood_Creaks_01}}
 {id=WavePlayer_57, properties {soundWaveAssetPtr=&Spot_DarkForest_Trees_Wood_Creaks_06}}
 {id=WavePlayer_58, properties {soundWaveAssetPtr=&Spot_DarkForest_Trees_Wood_Creaks_10}}
 {id=WavePlayer_59, properties {soundWaveAssetPtr=&Spot_DarkForest_Trees_Wood_Creaks_15}}
]
properties {
 attenuationSettings=/Game/Sounds/Attenuation/Fort_TileAmbient_OneShot_Distant_Creak_Att
 maxDistance=8200
 soundClassObject=&Fort_Ambient_OneShots_Outdoors
 volumeMultiplier=0.1
 busSends [{
   audioBus=null
   maxSendDistance=6000
   minSendDistance=3000
   minSendLevel=0.5
   sendLevel=1
   soundSourceBus=&AmbientAudio_Exterior_SourceBus
  }]
}
